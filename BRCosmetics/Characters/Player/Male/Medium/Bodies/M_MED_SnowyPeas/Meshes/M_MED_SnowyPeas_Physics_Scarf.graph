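# Physics asset for the scarf on the SnowyPeas medium male player cosmetic body.
# Capsules on the pelvis, spine, neck, clavicles, arms and thighs form the body volume around a five-link box chain, dyn_scarf_01 to 05.
# The chain hangs from a small dyn_scarf_base box constrained to spine_05.
# Body and base joints swing and twist freely, while scarf links have tight angular limits that generally widen toward the tip.
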PhysicsAsset /BRCosmetics/Characters/Player/Male/Medium/Bodies/M_MED_SnowyPeas/Meshes/M_MED_SnowyPeas_Physics_Scarf
bodies {
 clavicle_l [{shapeType=Capsule, length=4.3757, radius=5.6243, shapeName=clavicle_l_capsule, center=<8.778 0 0.7483>, rotation=<rot 90 0 0>}]
 clavicle_r [{shapeType=Capsule, length=4.3757, radius=5.6243, shapeName=clavicle_r_capsule, center=<-8.778 0 -0.7483>, rotation=<rot 90 180 0>}]
 dyn_scarf_01 [{shapeType=Box, extentX=3.2776, extentY=6, extentZ=12.0456, shapeName=dyn_scarf_01_box, center=<3.0522 0 0>, rotation=<rot 0 90 0>}]
 dyn_scarf_02 [{shapeType=Box, extentX=3.2776, extentY=6, extentZ=12.0456, shapeName=dyn_scarf_02_box, center=<3.0522 0 0>, rotation=<rot 0 90 0>}]
 dyn_scarf_03 [{shapeType=Box, extentX=3.2776, extentY=6, extentZ=12.0456, shapeName=dyn_scarf_03_box, center=<3.0522 0 0>, rotation=<rot 0 90 0>}]
 dyn_scarf_04 [{shapeType=Box, extentX=3.2776, extentY=6, extentZ=12.0456, shapeName=dyn_scarf_04_box, center=<3.0522 0 0>, rotation=<rot 0 90 0>}]
 dyn_scarf_05 [{shapeType=Box, extentX=3.2776, extentY=6, extentZ=12.0456, shapeName=dyn_scarf_05_box, center=<3.0522 0 0>, rotation=<rot 0 90 0>}]
 dyn_scarf_base [{shapeType=Box, extentX=1.01, extentY=1.01, extentZ=1.01, shapeName=dyn_scarf_base_box, rotation=<rot 0 -90 0>}]
 lowerarm_l [{shapeType=Capsule, length=22.3649, radius=7.6332, shapeName=lowerarm_l_capsule, center=<12.8776 0 0>, rotation=<rot 90 0 0>}]
 lowerarm_r [{shapeType=Capsule, length=22.3649, radius=7.6332, shapeName=lowerarm_r_capsule, center=<-12.8776 0 0>, rotation=<rot 90 180 0>}]
 neck_01 [{shapeType=Capsule, length=10.4343, radius=11.5252, shapeName=neck_01_capsule}]
 pelvis [{shapeType=Capsule, length=12.6188, radius=12.1677, shapeName=pelvis_capsule}]
 spine_01 [{shapeType=Capsule, length=10.114, radius=11.4362, shapeName=spine_01_capsule}]
 spine_02 [{shapeType=Capsule, length=9.2316, radius=12.5495, shapeName=spine_02_capsule}]
 spine_03 [{shapeType=Capsule, length=8.822, radius=14.2669, shapeName=spine_03_capsule, center=<0 -1.3575 0>}]
 spine_04 [{shapeType=Capsule, length=11.0831, radius=13.2909, shapeName=spine_04_capsule, center=<1.5275 -0.9613 0>}]
 spine_05 [{shapeType=Capsule, length=12.4496, radius=13.3076, shapeName=spine_05_capsule, center=<1.9882 -1.4264 0>}]
 thigh_l [{shapeType=Capsule, length=19.9028, radius=10.0879, shapeName=thigh_l_capsule, center=<-8.6118 0 0>, rotation=<rot 90 0 0>}]
 thigh_r [{shapeType=Capsule, length=19.9028, radius=10.0879, shapeName=thigh_r_capsule, center=<8.6118 0 0>, rotation=<rot 90 180 0>}]
 thigh_twist_01_l [{shapeType=Capsule, length=17.7231, radius=6.1016, shapeName=thigh_twist_01_l_capsule, center=<-8.4681 0 0>, rotation=<rot 90 0 0>}]
 thigh_twist_01_r [{shapeType=Capsule, length=17.7231, radius=6.1016, shapeName=thigh_twist_01_r_capsule, center=<8.4681 0 0>, rotation=<rot 90 180 0>}]
 upperarm_l [{shapeType=Capsule, length=24.4164, radius=5.5817, shapeName=upperarm_l_capsule, center=<11.0637 0 0>, rotation=<rot 90 0 0>}]
 upperarm_r [{shapeType=Capsule, length=24.4164, radius=5.5817, shapeName=upperarm_r_capsule, center=<-11.0637 0 0>, rotation=<rot 90 180 0>}]
}
constraints [
 {
  bone1Name=spine_01
  bone2Name=pelvis
  swing1LimitDegrees=45
  swing1Motion=Free
  swing2LimitDegrees=45
  swing2Motion=Free
  twistLimitDegrees=45
  twistMotion=Free
 }
 {
  bone1Name=spine_02
  bone2Name=spine_01
  swing1LimitDegrees=45
  swing1Motion=Free
  swing2LimitDegrees=45
  swing2Motion=Free
  twistLimitDegrees=45
  twistMotion=Free
 }
 {
  bone1Name=spine_03
  bone2Name=spine_02
  swing1LimitDegrees=45
  swing1Motion=Free
  swing2LimitDegrees=45
  swing2Motion=Free
  twistLimitDegrees=45
  twistMotion=Free
 }
 {
  bone1Name=spine_04
  bone2Name=spine_03
  swing1LimitDegrees=45
  swing1Motion=Free
  swing2LimitDegrees=45
  swing2Motion=Free
  twistLimitDegrees=45
  twistMotion=Free
 }
 {
  bone1Name=spine_05
  bone2Name=spine_04
  swing1LimitDegrees=45
  swing1Motion=Free
  swing2LimitDegrees=45
  swing2Motion=Free
  twistLimitDegrees=45
  twistMotion=Free
 }
 {
  bone1Name=clavicle_l
  bone2Name=spine_05
  swing1LimitDegrees=45
  swing1Motion=Free
  swing2LimitDegrees=45
  swing2Motion=Free
  twistLimitDegrees=45
  twistMotion=Free
 }
 {
  bone1Name=upperarm_l
  bone2Name=clavicle_l
  swing1LimitDegrees=45
  swing1Motion=Free
  swing2LimitDegrees=45
  swing2Motion=Free
  twistLimitDegrees=45
  twistMotion=Free
 }
 {
  bone1Name=clavicle_r
  bone2Name=spine_05
  swing1LimitDegrees=45
  swing1Motion=Free
  swing2LimitDegrees=45
  swing2Motion=Free
  twistLimitDegrees=45
  twistMotion=Free
 }
 {
  bone1Name=upperarm_r
  bone2Name=clavicle_r
  swing1LimitDegrees=45
  swing1Motion=Free
  swing2LimitDegrees=45
  swing2Motion=Free
  twistLimitDegrees=45
  twistMotion=Free
 }
 {
  bone1Name=lowerarm_r
  bone2Name=upperarm_r
  swing1LimitDegrees=45
  swing1Motion=Free
  swing2LimitDegrees=45
  swing2Motion=Free
  twistLimitDegrees=45
  twistMotion=Free
 }
 {
  bone1Name=lowerarm_l
  bone2Name=upperarm_l
  swing1LimitDegrees=45
  swing1Motion=Free
  swing2LimitDegrees=45
  swing2Motion=Free
  twistLimitDegrees=45
  twistMotion=Free
 }
 {
  bone1Name=neck_01
  bone2Name=spine_05
  swing1LimitDegrees=45
  swing1Motion=Free
  swing2LimitDegrees=45
  swing2Motion=Free
  twistLimitDegrees=45
  twistMotion=Free
 }
 {
  bone1Name=thigh_l
  bone2Name=pelvis
  swing1LimitDegrees=45
  swing1Motion=Free
  swing2LimitDegrees=45
  swing2Motion=Free
  twistLimitDegrees=45
  twistMotion=Free
 }
 {
  bone1Name=thigh_r
  bone2Name=pelvis
  swing1LimitDegrees=45
  swing1Motion=Free
  swing2LimitDegrees=45
  swing2Motion=Free
  twistLimitDegrees=45
  twistMotion=Free
 }
 {
  bone1Name=thigh_twist_01_l
  bone2Name=thigh_l
  swing1LimitDegrees=45
  swing1Motion=Free
  swing2LimitDegrees=45
  swing2Motion=Free
  twistLimitDegrees=45
  twistMotion=Free
 }
 {
  bone1Name=thigh_twist_01_r
  bone2Name=thigh_r
  swing1LimitDegrees=45
  swing1Motion=Free
  swing2LimitDegrees=45
  swing2Motion=Free
  twistLimitDegrees=45
  twistMotion=Free
 }
 {
  bone1Name=dyn_scarf_base
  bone2Name=spine_05
  swing1LimitDegrees=45
  swing1Motion=Free
  swing2LimitDegrees=45
  swing2Motion=Free
  twistLimitDegrees=45
  twistMotion=Free
 }
 {bone1Name=dyn_scarf_01, bone2Name=dyn_scarf_base, swing1LimitDegrees=5, swing2LimitDegrees=10, twistLimitDegrees=0}
 {bone1Name=dyn_scarf_02, bone2Name=dyn_scarf_01, swing1LimitDegrees=10, swing2LimitDegrees=10, twistLimitDegrees=5}
 {bone1Name=dyn_scarf_03, bone2Name=dyn_scarf_02, swing1LimitDegrees=10, swing2LimitDegrees=15, twistLimitDegrees=5}
 {bone1Name=dyn_scarf_04, bone2Name=dyn_scarf_03, swing1LimitDegrees=15, swing2LimitDegrees=20, twistLimitDegrees=5}
 {bone1Name=dyn_scarf_05, bone2Name=dyn_scarf_04, swing1LimitDegrees=10, swing2LimitDegrees=25, twistLimitDegrees=5}
]
