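# Physics asset for the front vest flaps on the M_MED_SnowyPeas male player body, named for secondary cloth-like motion via dynamic bones.
# Capsules on the pelvis, spine and thighs form the body volume, linked by free-rotation constraints.
# Left and right three-bone flap chains hang from spine_04 through tiny, freely constrained base capsules.
# Lower flap joints act as hinges, with one narrow swing range that widens at the tip and zero-degree limits on other axes.
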PhysicsAsset /BRCosmetics/Characters/Player/Male/Medium/Bodies/M_MED_SnowyPeas/Meshes/M_MED_SnowyPeas_Physics_Vest
bodies {
 dyn_vest_frnt_l_01 [{shapeType=Capsule, length=11.2366, radius=4.2299, shapeName=dyn_vest_frnt_l_01_capsule, center=<3.2128 2.5541 0>}]
 dyn_vest_frnt_l_02 [{shapeType=Capsule, length=11.3366, radius=4.4573, shapeName=dyn_vest_frnt_l_02_capsule, center=<2.7588 2.5541 0>}]
 dyn_vest_frnt_l_base [{shapeType=Capsule, length=0.505, radius=0.505, shapeName=dyn_vest_frnt_l_base_capsule, rotation=<rot 0 -90 0>}]
 dyn_vest_frnt_r_01 [{shapeType=Capsule, length=11.3774, radius=3.6639, shapeName=dyn_vest_frnt_r_01_capsule, center=<-3.7166 -2.5211 0>}]
 dyn_vest_frnt_r_02 [{shapeType=Capsule, length=11.8931, radius=4.5159, shapeName=dyn_vest_frnt_r_02_capsule, center=<-2.7275 -2.6868 0>}]
 dyn_vest_frnt_r_base [{shapeType=Capsule, length=0.505, radius=0.505, shapeName=dyn_vest_frnt_r_base_capsule, rotation=<rot 0 -90 0>}]
 pelvis [{shapeType=Capsule, length=12.6188, radius=12.1677, shapeName=pelvis_capsule}]
 spine_01 [{shapeType=Capsule, length=10.114, radius=11.4362, shapeName=spine_01_capsule}]
 spine_02 [{shapeType=Capsule, length=9.3105, radius=9.6713, shapeName=spine_02_capsule}]
 spine_03 [{shapeType=Capsule, length=6.7019, radius=10.9438, shapeName=spine_03_capsule}]
 spine_04 [{shapeType=Capsule, length=10.7913, radius=9.431, shapeName=spine_04_capsule}]
 thigh_l [{shapeType=Capsule, length=19.9028, radius=10.0879, shapeName=thigh_l_capsule, center=<-8.6118 0 0>, rotation=<rot 90 0 0>}]
 thigh_r [{shapeType=Capsule, length=19.9028, radius=10.0879, shapeName=thigh_r_capsule, center=<8.6118 0 0>, rotation=<rot 90 180 0>}]
}
constraints [
 {
  bone1Name=spine_01
  bone2Name=pelvis
  swing1LimitDegrees=45
  swing1Motion=Free
  swing2LimitDegrees=45
  swing2Motion=Free
  twistLimitDegrees=45
  twistMotion=Free
 }
 {
  bone1Name=spine_02
  bone2Name=spine_01
  swing1LimitDegrees=45
  swing1Motion=Free
  swing2LimitDegrees=45
  swing2Motion=Free
  twistLimitDegrees=45
  twistMotion=Free
 }
 {
  bone1Name=spine_03
  bone2Name=spine_02
  swing1LimitDegrees=45
  swing1Motion=Free
  swing2LimitDegrees=45
  swing2Motion=Free
  twistLimitDegrees=45
  twistMotion=Free
 }
 {
  bone1Name=spine_04
  bone2Name=spine_03
  swing1LimitDegrees=45
  swing1Motion=Free
  swing2LimitDegrees=45
  swing2Motion=Free
  twistLimitDegrees=45
  twistMotion=Free
 }
 {
  bone1Name=thigh_l
  bone2Name=pelvis
  swing1LimitDegrees=45
  swing1Motion=Free
  swing2LimitDegrees=45
  swing2Motion=Free
  twistLimitDegrees=45
  twistMotion=Free
 }
 {
  bone1Name=thigh_r
  bone2Name=pelvis
  swing1LimitDegrees=45
  swing1Motion=Free
  swing2LimitDegrees=45
  swing2Motion=Free
  twistLimitDegrees=45
  twistMotion=Free
 }
 {
  bone1Name=dyn_vest_frnt_l_base
  bone2Name=spine_04
  swing1LimitDegrees=45
  swing1Motion=Free
  swing2LimitDegrees=45
  swing2Motion=Free
  twistLimitDegrees=45
  twistMotion=Free
 }
 {bone1Name=dyn_vest_frnt_l_01, bone2Name=dyn_vest_frnt_l_base, swing1LimitDegrees=10, swing2LimitDegrees=0, twistLimitDegrees=0}
 {bone1Name=dyn_vest_frnt_l_02, bone2Name=dyn_vest_frnt_l_01, swing1LimitDegrees=20, swing2LimitDegrees=0, twistLimitDegrees=0}
 {
  bone1Name=dyn_vest_frnt_r_base
  bone2Name=spine_04
  swing1LimitDegrees=45
  swing1Motion=Free
  swing2LimitDegrees=45
  swing2Motion=Free
  twistLimitDegrees=45
  twistMotion=Free
 }
 {bone1Name=dyn_vest_frnt_r_01, bone2Name=dyn_vest_frnt_r_base, swing1LimitDegrees=10, swing2LimitDegrees=0, twistLimitDegrees=0}
 {bone1Name=dyn_vest_frnt_r_02, bone2Name=dyn_vest_frnt_r_01, swing1LimitDegrees=20, swing2LimitDegrees=0, twistLimitDegrees=0}
]
