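# Animation Blueprint for the SnowyPeas male medium face accessory mesh that adds secondary physics motion to its pompon.
# It copies the pose from the attached parent mesh, then runs a single Anim Dynamics node on dyn_pompon before output.
# The root-relative simulation uses an angular spring, damping override, tight angular limits, custom gravity and clamped component acceleration to keep the sway contained.
# The event graph has no logic and there are no variables.
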AnimBlueprint /BRCosmetics/Characters/Player/Male/Medium/Bodies/M_MED_SnowyPeas/Meshes/Parts/M_MED_SnowyPeas_FaceAcc_AnimBP
target_skeleton=&/M_MED_SnowyPeas_FaceAcc_Skeleton
graphs [
 AnimGraph {nodes [
   n1 = Anim Dynamics - Bone: dyn_pompon(ComponentPose=Local To Component(LocalPose=Copy Pose From Mesh() {bUseAttachedParent=true}), Alpha=1, ExternalForce="0,0,0") {
    angularDampingOverride=1
    angularSpringConstant=500
    bAngularSpring=true
    bOverrideAngularDamping=true
    bUseGravityOverride=true
    simulationSpace=RootRelative
    alphaBoolBlend {blendOption=Linear}
    boundBone {boneName=dyn_pompon}
    componentAppliedLinearAccClamp=<3000 3000 3000>
    gravityOverride {z=-980}
    physicsBodyDefinitions [{
      boundBone {boneName=dyn_pompon}
      boxExtents=<9 9 9>
      constraintSetup {coneAngle=1, angularLimitsMax=<10 10 10>, angularLimitsMin=<-10 -10 -10>, angularTarget=<1 0 0>}
      localJointOffset=<0.3 0 0>
     }]
    relativeSpaceBone {boneName=thigh_l}
    simSpaceSettings {maxAngularAcceleration=1500, maxAngularVelocity=1200, simSpaceAngularAlpha=1, externalAngularVelocity=<1 0.65>}
   }
   Output Pose(Result=Component To Local(ComponentPose=@n1)) {layerGroup=None}
  ]}
 EventGraph {nodes [Event BlueprintUpdateAnimation()]}
]
variables []
